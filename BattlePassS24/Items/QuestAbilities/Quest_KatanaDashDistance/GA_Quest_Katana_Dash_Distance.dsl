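; Quest gameplay ability that tracks distance covered by a katana dash, including the airborne travel afterwards, for quest progress.
; The triggering event's magnitude, converted to meters, seeds the initial distance, then a looping timer samples the owning pawn.
; While the pawn has the MovementMode.Falling tag, it accumulates horizontal (XY) distance between samples.
; Once that tag is absent, it sends the rounded-up total as a Katana.DashDistance stat event, clears the timer and ends.
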(event Ability|EventActivateAbilityFromEvent
  (params (EventData :type Gameplay%20Event%20Data%20Structure%20%28by%20ref%29))
  (bind (_eventtag _instigator _target _optionalobject _optionalobject2 _contexthandle _instigatortags _targettags _eventmagnitude) (Utilities|Struct|BreakGameplayEventData EventData))
  (|CallParentFunction EventData)
  (Variables|Default|SetInitial%20Distance (/ _eventmagnitude 1e+02))
  (Variables|Default|SetPreviousLocation (Transformation|GetActorLocation (Variables|GAVars|Data|GetOwningPlayerPawn)))
  ;[1 1 1] Start Checking Vehicle If Vehicle is in Air
  (bind _returnvalue (Utilities|Time|SetTimerbyEvent (delegate Custom|Check%20In%20Air) (Variables|Default|GetExecutionTimeInterval) true))
  (Variables|Default|SetTimer%20Handle _returnvalue))

(event Custom|Check%20In%20Air
  (bind (_previouslocation_x _previouslocation_y) (Math|Vector|BreakVector (Variables|GetPreviousLocation_X)))
  (bind (_currentlocation_x _currentlocation_y) (Math|Vector|BreakVector (Variables|GetCurrentLocation_X)))
  ;[1 1 1] Check if player is in the air
  (if (GameplayTags|HasTag (GameplayTags|GetOwnedGameplayTags (Variables|GAVars|Data|GetOwningPlayerPawn) :!class /Script/GameplayTags.GameplayTagAssetInterface) "(TagName=\"MovementMode.Falling\")")
    ;[1 1 1] If in the air/post slide, calculate distance traveled while in the air
    (Variables|Default|SetCurrentLocation (Transformation|GetActorLocation (Variables|GAVars|Data|GetOwningPlayerPawn)))
    (Variables|Default|SetMetersTraveled (+ (/ (Math|Vector|VectorLength (Math|Vector|vector-vector (Math|Vector|MakeVector _currentlocation_x _currentlocation_y) (Math|Vector|MakeVector _previouslocation_x _previouslocation_y))) 1e+02) (Variables|Default|GetMetersTraveled)))
    (Variables|Default|SetPreviousLocation (Variables|Default|GetCurrentLocation))
    (else
      ;[1 1 1] When we are done adding 'fall off' distance
      (Quest|SendComplexCustomStatEvent (Variables|GAVars|Data|GetOwningPlayerQuestManager) 0 "(GameplayTags=((TagName=\"Athena.Event.Gameplay.Katana.DashDistance\")))" "(GameplayTags=)" (Math|Float|Ceil (+ (Variables|Default|GetMetersTraveled) (Variables|Default|GetInitial%20Distance))))
      (Variables|Default|SetMetersTraveled 0.0)
      (Utilities|Time|ClearandInvalidateTimerbyHandle (Variables|Default|GetTimer%20Handle))
      (Ability|EndAbility))))
